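# Niagara ribbon trail for the InflatoDodo pickaxe swing, with a single Main_Trail emitter.
# Spawns three anim trails from bone or socket indices, with width and brightness on short emitter-age curves that peak and drop within 0.2 seconds.
# User.Variant selects the color, one option shifting with time of day; updates apply time-of-day tint and alpha scaled by swing fade.
# It also derives wider core width and age-gated core color, and scrolls ribbon U with engine time.
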NiagaraSystem /BRCosmetics/Weapons/FORT_Melee/Pickaxe_InflatoDodo/FX/NS_Pickaxe_InflatoDodo_Trail
dependencies {
 usedDataInterfaces [&NiagaraDataInterfaceCurve, &NiagaraDataInterfaceSocketReader]
 usedDynamicInputs []
 usedModules [
  &self:SystemSpawnScript.SystemScriptSource.SystemScriptGraph.NiagaraNodeAssignment_3.SetVariables_#1
  &SystemState
  &self:SystemSpawnScript.SystemScriptSource.SystemScriptGraph.NiagaraNodeAssignment_2.SetVariables_#2
  &FN_VariantManager
 ]
}
emitters {Main_Trail {
  emitterSpawnScript [&FN_RibbonIDOffsetUtility()]
  emitterUpdateScript [
   &EmitterState()
   &FN_SpawnAnimTrail(Number of Anim Trails=3)
  ]
  particleSpawnScript [
   &FN_InitializeAnimTrailBoneIndecies(Second Bone or Socket Index=0)
   &FN_InitializeAnimTrailBoneIndecies(AnimTrail Selection=1, First Bone or Socket Index=1) {module=FN_InitializeAnimTrailBoneIndecies001}
   &FN_InitializeAnimTrailBoneIndecies(AnimTrail Selection=2, First Bone or Socket Index=2, Second Bone or Socket Index=2) {module=FN_InitializeAnimTrailBoneIndecies002}
   &FN_InitializeAnimTrail(
    AnimTrailWidth=&FloatFromCurve(
     FloatCurve=DataInterface() {propertyValues {Curve {keys [
         {interpMode=RCIM_Cubic, tangentMode=RCTM_User, time=0, value=0}
         {interpMode=RCIM_Cubic, tangentMode=RCTM_User, time=0.125, value=1}
         {arriveTangent=-29.9067, tangentMode=RCTM_User, time=0.2, value=0}
        ]}}}
     CurveIndex=$Emitter.Age
     Scale Curve=4
    )
    Color=&Multiply_LinearColorByFloat(
     Linear Color=&MergeHelper_ColorSwitch(
      0=<rgba 1 0.1806 0.06 1>
      1=&FNTimeOfDayColor(
       Night Color=<rgba 0.0875 0.25 0.2365 0.5>
       Morning Color=<rgba 0.05 0.2 0.1875 0.7>
       Day Color=<rgba 0.05 0.2 0.1875 0.7>
       Evening Color=<rgba 0.05 0.2 0.1875 0.7>
      )
      2=?
      3=?
      Index=&MakeIntFromFloat(Float=$User.Variant)
      Num=::NewEnumerator0
     )
     Float=&FloatFromCurve(
      FloatCurve=DataInterface() {propertyValues {Curve {keys [
          {interpMode=RCIM_Cubic, tangentMode=RCTM_User, time=0, value=0}
          {interpMode=RCIM_Cubic, tangentMode=RCTM_User, time=0.125, value=1}
          {arriveTangent=-29.9067, tangentMode=RCTM_User, time=0.2, value=0}
         ]}}}
      CurveIndex=$Emitter.Age
      Scale Curve=1.5
     )
    )
    Lifetime=0.2
   )
  ]
  particleUpdateScript [
   &SolveForcesAndVelocity(Clamp Velocity=false, Limit Acceleration=false)
   &FN_UpdateAnimTrail() {module=FN_UpdateAnimTrail001}
   &ScaleColor(
    Scale Alpha=&FloatFromCurve(
     FloatCurve=DataInterface() {propertyValues {Curve {keys [
         {arriveTangent=3.5079, interpMode=RCIM_Cubic, leaveTangent=3.5079, tangentMode=RCTM_User, time=0, value=0.6}
         {arriveTangent=0.0521, interpMode=RCIM_Cubic, leaveTangent=0.0521, tangentMode=RCTM_User, time=0.4, value=1.6}
         {interpMode=RCIM_Cubic, tangentMode=RCTM_User, time=1, value=0}
        ]}}}
     Scale Curve=&Multiply_Float(
      A=$System.SwingGeneralFade
      B=/BRCosmetics/Weapons/FORT_Melee/Pickaxe_GateHound/FX/VariantManager/VariantFloatSelector(
       Original Float:NiagaraFloat=10
       Variant 1 Float:NiagaraFloat=15
       Variant 2 Float:NiagaraFloat=?
       Variant 3 Float:NiagaraFloat=?
       Variant Amount:NiagaraInt32=0
      )
     )
    )
    Scale RGB=&MakeVectorFromLinearColorRGB(Color=&FNTimeOfDayColor(
      Night Color=<rgba 0.4507 0.4928 0.6094 1>
      Morning Color=<rgba 0.7917 0.4346 0.4573 1>
      Day Color=<rgba 1 1 1 1>
      Evening Color=<rgba 0.7292 0.5203 0.5203 1>
     ))
    ScaleA=true
    ScaleRGB=true
   )
   /SetVariables(Particles.RibbonU0Override=&Multiply_Float(A=$Engine.Time, B=4)) {isSetParameters=true}
   /SetVariables(
    Particles.ColorCore:LinearColor=&MultiplyLinearColor(
     Linear Color=$Particles.Color
     Scale Factor=&MakeVector4(
      X=1
      Y=1
      Z=1
      W=&FloatFromCurve(FloatCurve=DataInterface() {propertyValues {Curve {keys [<at 0.13 0>,<at 0.2 1>]}}}, CurveIndex=$System.Age)
     )
    )
    Particles.RibbonWidthCore:NiagaraFloat=&Multiply_Float(A=$Particles.RibbonWidth, B=10)
   ) {isSetParameters=true}
  ]
  rendererClasses [&NiagaraRibbonRendererProperties]
  renderers [{rendererClass=&NiagaraRibbonRendererProperties}]
 }}
summary {
 systemName=NS_Pickaxe_InflatoDodo_Trail
 emitters [{emitterName=Main_Trail, rendererClasses [&NiagaraRibbonRendererProperties]}]
 userVariables [
  User.RibbonIDOffset {type=NiagaraInt32}
  User.Variant {type=NiagaraFloat, defaultValue=0, description=""}
 ]
}
