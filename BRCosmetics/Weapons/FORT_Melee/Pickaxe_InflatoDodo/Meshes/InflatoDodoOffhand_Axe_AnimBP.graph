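# Animation Blueprint for the InflatoDodo offhand pickaxe, tracking whether the equipped pickaxe has the Style.OptionA cosmetic tag.
# Each update sets Equipped from that tag check.
# A state machine switches between the one-shot equip animation and the reference pose on that flag.
# Its result is cached as 'Input pose cache', but the final output reads 'PostAnim', whose save node has no input connected.
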AnimBlueprint /BRCosmetics/Weapons/FORT_Melee/Pickaxe_InflatoDodo/Meshes/InflatoDodoOffhand_Axe_AnimBP
target_skeleton=&/InflatoDodo_Axe_Skeleton
graphs [
 AnimGraph {nodes [
   Output Pose(Result=Use cached pose 'PostAnim'()) {layerGroup=None}
   Save cached pose 'PostAnim'() {cacheName=PostAnim}
   AnimGraphNode_StateMachine_0 = Equip State() {is=title AnimGraphNode_StateMachine}
   Save cached pose 'Input pose cache'(Pose=@AnimGraphNode_StateMachine_0) {cacheName=Input pose cache}
   Slot 'DefaultSlot'()
   AnimGraphNode_SubInput_0 = Input Pose()
   Use cached pose 'Input pose cache'()
  ]}
 Equip State {
  entry=Equip
  path=AnimGraph.AnimGraphNode_StateMachine_0.Equip State
  ids=id
  nodes [
   AnimStateNode_0 = Equip() {is=title AnimStateNode}
   AnimStateNode_1 = Unequipped() {is=title AnimStateNode}
   AnimStateTransitionNode_0 = AnimStateTransitionNode()
   AnimStateTransitionNode_5 = AnimStateTransitionNode()
  ]
  links [
   {
    dst=Unequipped
    rule=AnimGraph.AnimGraphNode_StateMachine_0.Equip State.AnimStateTransitionNode_0.Transition
    src=Equip
    via=AnimStateTransitionNode_0
   }
   {
    dst=Equip
    rule=AnimGraph.AnimGraphNode_StateMachine_0.Equip State.AnimStateTransitionNode_5.Transition
    src=Unequipped
    via=AnimStateTransitionNode_5
   }
  ]
 }
 Equip {
  path=AnimGraph.AnimGraphNode_StateMachine_0.Equip State.AnimStateNode_0.Equip
  nodes [Output Animation Pose(Result=Sequence Player 'InFlatoDodo_Offhand_equip'() {bLoopAnimation=false, sequence=/BRCosmetics/Weapons/FORT_Melee/Pickaxe_InflatoDodo/Anim/InFlatoDodo_Offhand_equip})]
 }
 Unequipped {
  path=AnimGraph.AnimGraphNode_StateMachine_0.Equip State.AnimStateNode_1.Unequipped
  nodes [Output Animation Pose(Result=Local Space Ref Pose())]
 }
 Transition {
  path=AnimGraph.AnimGraphNode_StateMachine_0.Equip State.AnimStateTransitionNode_0.Transition
  nodes [Result(bCanEnterTransition=NOT Boolean(A=Get Equipped() -> (Equipped:bool)))]
 }
 Transition {
  path=AnimGraph.AnimGraphNode_StateMachine_0.Equip State.AnimStateTransitionNode_5.Transition
  nodes [Result(bCanEnterTransition=Get Equipped())]
 }
 EventGraph {nodes [
   Event BlueprintUpdateAnimation()
   > n1 = Branch(Condition=PickaxeItemDefinitionHasAnyMetaTag(Tags=Make Array("[0]"=(TagName="Cosmetics.MetaTags.Style.OptionA")), bRequiresEquipped=true))
   > Set Equipped(Equipped:bool=true) -> (Output_Get:bool)
   Set Equipped(execute=@n1.else, Equipped=false)
   TryGetPawnOwner()
  ]}
]
variables [Equipped {type=boolean}]
